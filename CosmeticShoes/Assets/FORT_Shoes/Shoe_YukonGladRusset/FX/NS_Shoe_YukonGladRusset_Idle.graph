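# Niagara system for the YukonGladRusset shoe cosmetic that puts dirt, smoke and ribbon trails at the character's feet.
# A system Timeline value ramps sprite spawn rates and ribbon alpha; the system also references running-check logic.
# BigDirt, DustSmoke and Smoke spawn at skeletal bones or sockets by partition, inherit reversed velocity, and scale by per-foot ShoeScale.
# Trail and TrailLow each draw two skeletal-attached ribbons, and every emitter is tinted by time-of-day color.
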NiagaraSystem /CosmeticShoes/Assets/FORT_Shoes/Shoe_YukonGladRusset/FX/NS_Shoe_YukonGladRusset_Idle
dependencies {
 usedDataInterfaces [
  &NiagaraDataInterfaceCurve
  &NiagaraDataInterfaceSpriteRendererInfo
  &NiagaraDataInterfaceColorCurve
  &NiagaraDataInterfaceSkeletalMesh
 ]
 usedDynamicInputs [
  /Game/Effects/Niagara/DynamicInputs/Logic/FN_IsRunningCheck
  &Boolean_AndOperation
  &SetBoolByFloatComparison
  &VectorLength
  &TransformVector
  &Boolean_NotOperation
  &FNTimeOfDayColor
 ]
 usedModules [
  &self:SystemSpawnScript.SystemScriptSource.SystemScriptGraph.NiagaraNodeAssignment_0.SetVariables_#1
  &self:SystemSpawnScript.SystemScriptSource.SystemScriptGraph.NiagaraNodeAssignment_1.SetVariables_#2
  &SystemState
  &self:SystemSpawnScript.SystemScriptSource.SystemScriptGraph.NiagaraNodeAssignment_2.SetVariables_#3
  &self:SystemSpawnScript.SystemScriptSource.SystemScriptGraph.NiagaraNodeAssignment_5.SetVariables_#4
  &self:SystemSpawnScript.SystemScriptSource.SystemScriptGraph.NiagaraNodeAssignment_6.SetVariables_#5
  &self:SystemSpawnScript.SystemScriptSource.SystemScriptGraph.NiagaraNodeAssignment_4.SetVariables_#6
  &Timeline
  &self:SystemSpawnScript.SystemScriptSource.SystemScriptGraph.NiagaraNodeAssignment_3.SetVariables_#7
 ]
}
emitters {
 BigDirt {
  emitterUpdateScript [
   &EmitterState()
   &SpawnRate(
    SpawnRate=&FloatFromCurve(
     FloatCurve=DataInterface() {propertyValues {Curve {keys [<at 0 0>,<at 0.3 0>,<at 0.9 1>]}}}
     CurveIndex=$System.Timeline.Time
     Scale Curve=85
    )
    Use Spawn Probability=false
   )
   &SpawnPerUnit(
    Module.Use Max Movement Threshold=true
    Module.Use Movement Tolerance=false
    Spawn Spacing=10
    Use Spawn Probability=false
    Velocity Vector:Vector3f=&Multiply_Vector(A=$Engine.Owner.Velocity, B=<1 1 0>)
   ) {enabled=false}
  ]
  particleSpawnScript [
   &PartitionParticles()
   &InitializeParticle(
    Color=<rgba 0.4566 0.376 0.3052 0.9>
    Lifetime Max=0.25
    Lifetime Min=0.15
    Lifetime Mode=::NewEnumerator1
    Sprite Rotation Mode=::NewEnumerator2
    Sprite Size Mode=::NewEnumerator1
    Uniform Sprite Size Max=11
    Uniform Sprite Size Min=8
    UsePositionOffset=false
    Write Material Random=true
   )
   &SkeletalMeshLocation(
    Bone / Socket Index=$Particles.PartitionParticles.Partition
    Bone Sampling Mode=::NewEnumerator4
    MaskSpawnGroup=false
    Sampled Position Offset=<0 0 -5>
    Skeletal Mesh=$System.Skeletal Mesh
   )
   &ShapeLocation(MaskSpawnGroup=false, Override Seed=false, Sphere Radius=12, Sphere Surface Distribution=0.5278, UseOwnerScale=true)
   &AddVelocity(Velocity=&RandomRangeVector(Minimum=<0 -40 0>, Maximum=<0 -120 0>)) {enabled=false}
   &InheritVelocity(Inherited Velocity Amount Scale=<-0.4 -0.4 0>, Speed Limit=true, UseSpeedThreshold=false)
   &SubUVAnimation(
    "Sprite Renderer "=DataInterface() {propertyValues {SpriteRenderer=&self:BigDirt_0.NiagaraSpriteRendererProperties_0}}
    SubUV Animation Mode:ENiagara_SubUVLookupModeV2=::NewEnumerator2
    UseEndFrame=false
    UseStartFrame=false
   )
   &Color(Color=&MultiplyLinearColor(Linear Color=$Particles.Color, Scale Factor=&MakeVector4_FromLinearColor(LinearColor=$System.TODColor)))
  ]
  particleUpdateScript [
   &ParticleState()
   &ScaleSpriteSize(
    Uniform Curve Scale=&Multiply_Float(
     A=1
     B=&MaskFloatBySpawnGroup(
      Masked Float=&MakeFloatFromVector(Vector=$User.ShoeScale_L, Channel=::NewEnumerator0)
      Passthrough Float=&MakeFloatFromVector(Vector=$User.ShoeScale_R, Channel=::NewEnumerator0)
      Spawn Group Mask=$Particles.PartitionParticles.Partition
     )
    )
    Uniform Curve Sprite Scale=DataInterface() {propertyValues {Curve {keys [
        {tangentMode=RCTM_User, time=0, value=0}
        <at 0.2 0.85>
        {interpMode=RCIM_Cubic, leaveTangent=-3.1447, tangentMode=RCTM_Break, time=0.6, value=1}
        {tangentMode=RCTM_User, time=1, value=0}
       ]}}}
   ) {module=ScaleSpriteSize001}
   &ScaleSpriteSize(
    Uniform Curve Scale=&Multiply_Float(
     A=1
     B=&MaskFloatBySpawnGroup(
      Masked Float=&MakeFloatFromVector(Vector=$User.ShoeScale_L, Channel=::NewEnumerator0)
      Passthrough Float=&MakeFloatFromVector(Vector=$User.ShoeScale_R, Channel=::NewEnumerator0)
      Spawn Group Mask=$Particles.PartitionParticles.Partition
     )
    )
    Uniform Curve Sprite Scale=DataInterface() {propertyValues {Curve {
       defaultValue=0
       keys [
        {arriveTangent=9.0725, interpMode=RCIM_Cubic, leaveTangent=9.0725, tangentMode=RCTM_User, time=0, value=0}
        {arriveTangent=-2.5249, interpMode=RCIM_Cubic, leaveTangent=-2.5249, tangentMode=RCTM_User, time=0.1408, value=0.9847}
        {arriveTangent=-0.0405, interpMode=RCIM_Cubic, leaveTangent=-0.0405, tangentMode=RCTM_User, time=1, value=0}
       ]
      }}}
   )
   &Drag(Drag=2, Use Linear Drag=true, Use Rotational Drag=true)
   &GravityForce(Gravity=&RandomRangeVector(Minimum=<0 0 -300>, Maximum=<0 0 -500>))
   &SolveForcesAndVelocity(Clamp Velocity=false, Limit Acceleration=false)
  ]
  rendererClasses [&NiagaraSpriteRendererProperties]
  renderers [{rendererClass=&NiagaraSpriteRendererProperties}]
 }
 DustSmoke {
  emitterUpdateScript [
   &EmitterState()
   &SpawnRate(
    SpawnRate=&FloatFromCurve(
     FloatCurve=DataInterface() {propertyValues {Curve {keys [<at 0 0>,<at 0.4 0>,<at 0.9 1>]}}}
     CurveIndex=$System.Timeline.Time
     Scale Curve=45
    )
    Use Spawn Probability=false
   ) {module=SpawnRate001}
   &SpawnRate(SpawnRate=50, Use Spawn Probability=false) {enabled=false}
   &SpawnPerUnit(
    Module.Use Max Movement Threshold=true
    Module.Use Movement Tolerance=false
    Spawn Spacing=25
    Use Spawn Probability=false
    Velocity Vector:Vector3f=&Multiply_Vector(A=$Engine.Owner.Velocity, B=<1 1 0>)
   ) {enabled=false}
  ]
  particleSpawnScript [
   &PartitionParticles()
   &InitializeParticle(
    Color=<rgba 0.4983 0.4304 0.3676 1>
    Lifetime Max=0.25
    Lifetime Min=0.12
    Lifetime Mode=::NewEnumerator1
    Sprite Rotation Mode=::NewEnumerator2
    Sprite Size Mode=::NewEnumerator1
    Uniform Sprite Size Max=50
    Uniform Sprite Size Min=35
    UsePositionOffset=false
    Write Material Random=true
   )
   &SkeletalMeshLocation(
    Bone / Socket Index=$Particles.PartitionParticles.Partition
    Bone Sampling Mode=::NewEnumerator4
    MaskSpawnGroup=false
    Sampled Position Offset=<0 0 -5>
    Skeletal Mesh=$System.Skeletal Mesh
   )
   &ShapeLocation(MaskSpawnGroup=false, Override Seed=false, Sphere Radius=5, Sphere Surface Distribution=0.3452, UseOwnerScale=true)
   &AddVelocity(Velocity=&RandomRangeVector(Minimum=<0 -40 0>, Maximum=<0 -120 0>)) {enabled=false}
   &InheritVelocity(Inherited Velocity Amount Scale=<-0.6 -0.6 0>, Speed Limit=true, UseSpeedThreshold=false)
   &Color(Color=&MultiplyLinearColor(Linear Color=$Particles.Color, Scale Factor=&MakeVector4_FromLinearColor(LinearColor=$System.TODColor)))
  ]
  particleUpdateScript [
   &ParticleState()
   &ScaleSpriteSize(
    Uniform Curve Scale=&Multiply_Float(
     A=1
     B=&MaskFloatBySpawnGroup(
      Masked Float=&MakeFloatFromVector(Vector=$User.ShoeScale_L, Channel=::NewEnumerator0)
      Passthrough Float=&MakeFloatFromVector(Vector=$User.ShoeScale_R, Channel=::NewEnumerator0)
      Spawn Group Mask=$Particles.PartitionParticles.Partition
     )
    )
    Uniform Curve Sprite Scale=DataInterface() {propertyValues {Curve {keys [
        {interpMode=RCIM_Cubic, leaveTangent=4.0079, tangentMode=RCTM_Break, time=0, value=0}
        {interpMode=RCIM_Cubic, tangentMode=RCTM_Break, time=0.7, value=1}
        {arriveTangent=0.452, tangentMode=RCTM_User, time=1, value=1.2}
       ]}}}
   )
   &Drag(Drag=4, Use Linear Drag=true, Use Rotational Drag=true)
   &GravityForce(Gravity=&RandomRangeVector(Minimum=<0 0 -450>, Maximum=<0 0 -700>)) {enabled=false}
   &SolveForcesAndVelocity(Clamp Velocity=false, Limit Acceleration=false)
   &ScaleSpriteSizeBySpeed(Max Scale Factor=<1.5 4>, Min Scale Factor=<0.8 2>)
   &DynamicMaterialParameters(
    Index 0 Param 1=&FloatFromCurve(FloatCurve=DataInterface() {propertyValues {Curve {keys [<at 0 1>,<at 1 0>]}}})
    Index 0 Param 4=1
    Module.Index0Float0Write=true
    Module.Index0Float1Write=true
    Module.Index0Float2Write=true
    Module.Index0Float3Write=true
   ) {enabled=false}
   &ScaleColor(
    Linear Color Curve=DataInterface() {propertyValues {
      AlphaCurve {keys [<at 0.1476 0.8>,<at 0.9652 0>]}
      BlueCurve {keys [<at 0 0.4226>,<at 0.8297 0.4242>]}
      GreenCurve {keys [<at 0 0.502>,<at 0.8297 0.5521>]}
      RedCurve {keys [<at 0 0.6163>,<at 0.8297 0.7274>]}
     }}
    Scale Mode=::NewEnumerator2
   )
  ]
  rendererClasses [&NiagaraSpriteRendererProperties]
  renderers [{rendererClass=&NiagaraSpriteRendererProperties}]
 }
 Smoke {
  emitterUpdateScript [
   &EmitterState()
   &SpawnRate(
    SpawnRate=&FloatFromCurve(
     FloatCurve=DataInterface() {propertyValues {Curve {keys [<at 0 0>,<at 0.4 0>,<at 0.9 1>]}}}
     CurveIndex=$System.Timeline.Time
     Scale Curve=50
    )
    Use Spawn Probability=false
   ) {module=SpawnRate001}
   &SpawnRate(SpawnRate=50, Use Spawn Probability=false) {enabled=false}
   &SpawnPerUnit(
    Module.Use Max Movement Threshold=true
    Module.Use Movement Tolerance=false
    Spawn Spacing=25
    Use Spawn Probability=false
    Velocity Vector:Vector3f=&Multiply_Vector(A=$Engine.Owner.Velocity, B=<1 1 0>)
   ) {enabled=false}
  ]
  particleSpawnScript [
   &PartitionParticles()
   &InitializeParticle(
    Color=<rgba 0.5747 0.4964 0.424 0.12>
    Lifetime Max=0.3
    Lifetime Min=0.2
    Lifetime Mode=::NewEnumerator1
    Sprite Size Mode=::NewEnumerator1
    Uniform Sprite Size Max=45
    Uniform Sprite Size Min=20
    UsePositionOffset=false
    Write Material Random=true
   )
   &SkeletalMeshLocation(
    Bone / Socket Index=$Particles.PartitionParticles.Partition
    Bone Sampling Mode=::NewEnumerator4
    MaskSpawnGroup=false
    Sampled Position Offset=<0 0 -5>
    Skeletal Mesh=$System.Skeletal Mesh
   )
   &ShapeLocation(MaskSpawnGroup=false, Override Seed=false, Sphere Radius=5, Sphere Surface Distribution=0.3452, UseOwnerScale=true)
   &AddVelocity(Velocity=&RandomRangeVector(Minimum=<0 -40 0>, Maximum=<0 -120 0>)) {enabled=false}
   &InheritVelocity(Inherited Velocity Amount Scale=<-0.6 -0.6 0>, Speed Limit=true, UseSpeedThreshold=false)
   &Color(Color=&MultiplyLinearColor(Linear Color=$Particles.Color, Scale Factor=&MakeVector4_FromLinearColor(LinearColor=$System.TODColor)))
  ]
  particleUpdateScript [
   &ParticleState()
   &ScaleSpriteSize(
    Uniform Curve Scale=&Multiply_Float(
     A=1
     B=&MaskFloatBySpawnGroup(
      Masked Float=&MakeFloatFromVector(Vector=$User.ShoeScale_L, Channel=::NewEnumerator0)
      Passthrough Float=&MakeFloatFromVector(Vector=$User.ShoeScale_R, Channel=::NewEnumerator0)
      Spawn Group Mask=$Particles.PartitionParticles.Partition
     )
    )
    Uniform Curve Sprite Scale=DataInterface() {propertyValues {Curve {keys [
        {interpMode=RCIM_Cubic, leaveTangent=4.0079, tangentMode=RCTM_Break, time=0, value=0}
        {interpMode=RCIM_Cubic, tangentMode=RCTM_Break, time=0.7, value=1}
        {arriveTangent=0.452, tangentMode=RCTM_User, time=1, value=1.2}
       ]}}}
   )
   &Drag(Drag=4, Use Linear Drag=true, Use Rotational Drag=true)
   &GravityForce(Gravity=&RandomRangeVector(Minimum=<0 0 -450>, Maximum=<0 0 -700>)) {enabled=false}
   &SolveForcesAndVelocity(Clamp Velocity=false, Limit Acceleration=false)
   &DynamicMaterialParameters(
    Index 0 Param 1=&FloatFromCurve(FloatCurve=DataInterface() {propertyValues {Curve {keys [<at 0 1>,<at 1 0>]}}})
    Index 0 Param 4=1
    Index 2 Param 1=&FloatFromCurve(FloatCurve=DataInterface() {propertyValues {Curve {keys [{interpMode=RCIM_Cubic, leaveTangent=3.3442, tangentMode=RCTM_Break, time=0, value=-0.05}, <at 0.7 1>]}}})
    Index 2 Param 3=&FloatFromCurve(
     FloatCurve=DataInterface() {propertyValues {Curve {keys [{interpMode=RCIM_Cubic, leaveTangent=2.4944, tangentMode=RCTM_Break, time=0, value=0}, <at 0.8 1>]}}}
     Scale Curve=2
    )
    Index 2 Param 4=&FloatFromCurve(FloatCurve=DataInterface() {propertyValues {Curve {keys [<at 0 0.25>,<at 0.5 1.05>]}}}, Scale Curve=0.7)
    Module.Index0Float0Write=true
    Module.Index0Float1Write=true
    Module.Index0Float2Write=true
    Module.Index0Float3Write=true
    Module.Index2Float0Write=true
    Module.Index2Float1Write=true
    Module.Index2Float2Write=true
    Module.Index2Float3Write=true
    Write Parameter Index 2=::NewEnumerator1
   )
   &ScaleColor(
    Linear Color Curve=DataInterface() {propertyValues {
      AlphaCurve {keys [<at 0.1743 1>,<at 0.9642 0>]}
      BlueCurve {keys [<at 0 1>,<at 0.8612 0.4242>]}
      GreenCurve {keys [<at 0 1>,<at 0.8612 0.5521>]}
      RedCurve {keys [<at 0 1>,<at 0.8612 0.7274>]}
     }}
    Scale Mode=::NewEnumerator2
   )
   &ScaleSpriteSizeBySpeed(Max Scale Factor=<1.5 3>, Min Scale Factor=<0.8 1.5>)
  ]
  rendererClasses [&NiagaraSpriteRendererProperties]
  renderers [{rendererClass=&NiagaraSpriteRendererProperties}]
 }
 Trail {
  emitterUpdateScript [
   &EmitterState()
   &FN_SpawnRibbon_v2(Number of Ribbons=2, Spawn Mode=::NewEnumerator0, SpawnRate=64)
   &FN_Integrator() {enabled=false}
  ]
  particleSpawnScript [
   /Game/Effects/Niagara/Modules/Ribbons/FN_InitializeRibbon_v4(
    Color:LinearColor=<rgba 0.8698 0.8227 0.75 0.25>
    Enable Skeletal Position=true
    Lifetime=0.21
    Ribbon Width=25
    SkeletalMesh:NiagaraDataInterfaceSkeletalMesh=$System.Skel2
   )
   &FN_FlipRibbonV()
   &AddVelocity(Velocity=<0 -130 0>)
  ]
  particleUpdateScript [
   &SolveForcesAndVelocity(Clamp Velocity=false, Limit Acceleration=false)
   &InheritSourceMovement(
    Applied Position Scale=&VectorFromFloat(Value=&FloatFromCurve(FloatCurve=DataInterface() {propertyValues {Curve {keys [
          {interpMode=RCIM_Cubic, tangentMode=RCTM_User, time=0, value=1}
          {arriveTangent=-0.0657, interpMode=RCIM_Cubic, leaveTangent=-0.0657, tangentMode=RCTM_User, time=1, value=0.14}
         ]}}}))
    ApplyForce=false
    ApplyPosition=true
    ApplyPositionFalloff=false
   ) {enabled=false}
   /Game/Effects/Niagara/Modules/Ribbons/FN_UpdateRibbon_v4(Enable Skeletal Position=true, SkeletalMesh:NiagaraDataInterfaceSkeletalMesh=$System.Skel2)
   &FN_RibbonMaterial() {enabled=false}
   &ScaleColor(
    Scale Alpha=&FloatFromCurve(
     FloatCurve=DataInterface() {propertyValues {Curve {keys [{interpMode=RCIM_Cubic, leaveTangent=-3.0223, tangentMode=RCTM_Break, time=0, value=1}, <at 1 0>]}}}
     Scale Curve=$System.Timeline.Time
    )
    Scale RGB=&MakeVectorFromLinearColorRGB(Color=&ColorFromCurve(ColorCurve=DataInterface() {propertyValues {
        AlphaCurve {keys []}
        BlueCurve {keys [<at 0.0628 0.8585 RCIM_Cubic>,<at 0.8496 0.1572 RCIM_Cubic>]}
        GreenCurve {keys [<at 0.0628 0.9212 RCIM_Cubic>,<at 0.8496 0.2317 RCIM_Cubic>]}
        RedCurve {keys [<at 0.0628 1 RCIM_Cubic>,<at 0.8496 0.3698 RCIM_Cubic>]}
       }}))
    ScaleA=true
    ScaleRGB=true
   )
   &ScaleRibbonWidth(Ribbon Width Scale=&FloatFromCurve(
     FloatCurve=DataInterface() {propertyValues {Curve {keys [
         {interpMode=RCIM_Cubic, leaveTangent=0.6658, tangentMode=RCTM_Break, time=0, value=1}
         {arriveTangent=1.4789, tangentMode=RCTM_User, time=1, value=2}
        ]}}}
     Scale Curve=&MaskFloatBySpawnGroup(
      Masked Float=&MakeFloatFromVector(Vector=$User.ShoeScale_L, Channel=::NewEnumerator0)
      Passthrough Float=&MakeFloatFromVector(Vector=$User.ShoeScale_R, Channel=::NewEnumerator0)
     )
    ))
   &DynamicMaterialParameters(
    Index 0 Param 1=&FloatFromCurve(FloatCurve=DataInterface() {propertyValues {Curve {keys [<at 0 1>,<at 0.5 0.3>,<at 1 0.05>]}}})
    Index 0 Param 2=0.2
    Index 0 Param 3=-2
    Index 0 Param 4=0.7
    Module.Index0Float0Write=true
    Module.Index0Float1Write=true
    Module.Index0Float2Write=true
    Module.Index0Float3Write=true
   )
   &Color(Color=&MultiplyLinearColor(Linear Color=$Particles.Color, Scale Factor=&MakeVector4_FromLinearColor(LinearColor=$System.TODColor)))
  ]
  rendererClasses [&NiagaraRibbonRendererProperties]
  renderers [{rendererClass=&NiagaraRibbonRendererProperties}]
 }
 TrailLow {
  emitterUpdateScript [
   &EmitterState()
   &FN_SpawnRibbon_v2(Number of Ribbons=2, Spawn Mode=::NewEnumerator0, SpawnRate=64)
   &FN_Integrator() {enabled=false}
  ]
  particleSpawnScript [
   &FN_InitializeRibbon_v4(
    Color:LinearColor=<rgba 0.8698 0.8227 0.75 0.25>
    Enable Skeletal Position=true
    Lifetime=0.21
    Ribbon Width=25
    SkeletalMesh:NiagaraDataInterfaceSkeletalMesh=$System.Skel2
   )
   &FN_FlipRibbonV()
   &AddVelocity(Velocity=<0 -130 0>)
  ]
  particleUpdateScript [
   &SolveForcesAndVelocity(Clamp Velocity=false, Limit Acceleration=false)
   &InheritSourceMovement(
    Applied Position Scale=&VectorFromFloat(Value=&FloatFromCurve(FloatCurve=DataInterface() {propertyValues {Curve {keys [
          {interpMode=RCIM_Cubic, tangentMode=RCTM_User, time=0, value=1}
          {arriveTangent=-0.0657, interpMode=RCIM_Cubic, leaveTangent=-0.0657, tangentMode=RCTM_User, time=1, value=0.14}
         ]}}}))
    ApplyForce=false
    ApplyPosition=true
    ApplyPositionFalloff=false
   ) {enabled=false}
   &FN_UpdateRibbon_v4(Enable Skeletal Position=true, SkeletalMesh:NiagaraDataInterfaceSkeletalMesh=$System.Skel2)
   &FN_RibbonMaterial() {enabled=false}
   &ScaleColor(
    Scale Alpha=&FloatFromCurve(
     FloatCurve=DataInterface() {propertyValues {Curve {keys [{interpMode=RCIM_Cubic, leaveTangent=-3.0223, tangentMode=RCTM_Break, time=0, value=1}, <at 1 0>]}}}
     Scale Curve=$System.Timeline.Time
    )
    Scale RGB=&MakeVectorFromLinearColorRGB(Color=&ColorFromCurve(ColorCurve=DataInterface() {propertyValues {
        AlphaCurve {keys []}
        BlueCurve {keys [<at 0.0628 0.8585 RCIM_Cubic>,<at 0.8496 0.1572 RCIM_Cubic>]}
        GreenCurve {keys [<at 0.0628 0.9212 RCIM_Cubic>,<at 0.8496 0.2317 RCIM_Cubic>]}
        RedCurve {keys [<at 0.0628 1 RCIM_Cubic>,<at 0.8496 0.3698 RCIM_Cubic>]}
       }}))
    ScaleA=true
    ScaleRGB=true
   )
   &ScaleRibbonWidth(Ribbon Width Scale=&FloatFromCurve(FloatCurve=DataInterface() {propertyValues {Curve {keys [
         {interpMode=RCIM_Cubic, leaveTangent=0.6658, tangentMode=RCTM_Break, time=0, value=1}
         {arriveTangent=1.4789, tangentMode=RCTM_User, time=1, value=2}
        ]}}}))
   &DynamicMaterialParameters(
    Index 0 Param 1=&FloatFromCurve(FloatCurve=DataInterface() {propertyValues {Curve {keys [<at 0 1>,<at 0.5 0.3>,<at 1 0.05>]}}})
    Index 0 Param 2=0.2
    Index 0 Param 3=-2
    Index 0 Param 4=0.3
    Module.Index0Float0Write=true
    Module.Index0Float1Write=true
    Module.Index0Float2Write=true
    Module.Index0Float3Write=true
   )
   &Color(Color=&MultiplyLinearColor(Linear Color=$Particles.Color, Scale Factor=&MakeVector4_FromLinearColor(LinearColor=$System.TODColor)))
  ]
  rendererClasses [&NiagaraRibbonRendererProperties]
  renderers [{rendererClass=&NiagaraRibbonRendererProperties}]
 }
}
summary {
 systemName=NS_Shoe_YukonGladRusset_Idle
 emitters [
  {emitterName=BigDirt, rendererClasses [&NiagaraSpriteRendererProperties]}
  {emitterName=Smoke, rendererClasses [&NiagaraSpriteRendererProperties]}
  {emitterName=Trail, rendererClasses [&NiagaraRibbonRendererProperties]}
  {emitterName=DustSmoke, rendererClasses [&NiagaraSpriteRendererProperties]}
  {emitterName=TrailLow, rendererClasses [&NiagaraRibbonRendererProperties]}
 ]
 userVariables [
  User.HasGameplayAbilityActive {type=NiagaraBool}
  User.HideFootFX {type=NiagaraBool}
  User.IsSkydiving {type=NiagaraBool}
  User.RibbonIDOffset {type=NiagaraInt32}
  User.ShoeScale_L {type=Vector3f, description="", defaultValue {struct=Vector3f, value=<0 0 0>}}
  User.ShoeScale_R {type=Vector3f, description="", defaultValue {struct=Vector3f, value=<0 0 0>}}
 ]
}
